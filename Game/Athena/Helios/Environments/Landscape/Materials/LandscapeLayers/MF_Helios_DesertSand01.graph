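# Helios desert sand landscape layer function that height-blends wavy sand into the incoming material under the BiomeA layer, with layer mask and static mesh outputs.
# A flow texture distorts the sand texture UVs, and height lerps layer five biome colors to tint the albedo.
# Diffuse alpha drives opacity mask and displacement.
# Mobile paths switch the mask and derive static mesh normals, while MobileBlendedMaterial bypasses the mobile blend.
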MaterialFunction /Game/Athena/Helios/Environments/Landscape/Materials/LandscapeLayers/MF_Helios_DesertSand01
graph {nodes [
  Constant() {r=1}
  Constant() {r=3}
  Constant() {r=2}
  n1 = NamedRerouteUsage()
  n2 = NamedRerouteUsage()
  n3 = NamedRerouteUsage()
  n4 = NamedRerouteUsage()
  n5 = NamedRerouteUsage()
  n6 = MaterialFunctionCall!(InputMaterial=@n1, LayerMaterial=@n2, LayerWeight=@n3, UseContrast=@n4, Contrast=@n5) {materialFunction=/Game/Athena/Helios/Environments/Landscape/Materials/MaterialFunctions_Utilities/MF_HeightBlendLandscapeLayers_OpacityMaskDisplaceCustom}
  FunctionOutput(None=LandscapeLayerSwitch!(LayerUsed=@n6.BlendedMaterial, LayerNotUsed=@n1) {parameterName=BiomeA}) {outputName=BlendedMaterial, sortPriority=-300}
  FunctionOutput(None=ShadingPathSwitch(Default=NamedRerouteUsage(), Mobile=NamedRerouteUsage())) {outputName=LayerMask, sortPriority=-1}
  n7 = NamedRerouteUsage()
  FunctionOutput(None=@n7) {outputName=MobileBlendedMaterial, sortPriority=-100}
  n8 = NamedRerouteUsage()
  n9 = StaticBool()
  FunctionOutput(None=ShadingPathSwitch(
    Default=NamedRerouteUsage()
    Mobile=MaterialFunctionCall!(MobileMaterial=@n8, TransformTangents=@n9) {materialFunction=/Game/Athena/Helios/Environments/Landscape/Materials/MaterialFunctions_Utilities/MF_MobileDeriveNormals}.ResultMaterial
   )) {outputName=StaticMeshMaterial, sortPriority=10}
  n10 = MaterialFunctionCall!(
   InputMaterial=@n7
   LayerMaterial=NamedRerouteUsage()
   LayerWeight=NamedRerouteUsage()
   UseContrast=NamedRerouteUsage()
   Contrast=NamedRerouteUsage()
  ) {materialFunction=/Game/Athena/Helios/Environments/Landscape/Materials/MaterialFunctions_Utilities/MF_HeightBlendLandscapeLayers_OpacityMask}
  LandscapeLayerSwitch(LayerUsed=@n10.BlendedMaterial, LayerNotUsed=@n7) {parameterName=BiomeA}
  MaterialFunctionCall(InputMaterial=@n1, LayerMaterial=@n2, LayerWeight=@n3, UseContrast=@n4, Contrast=@n5) {materialFunction=&MF_HeightBlendLandscapeLayers_OpacityMask}
  n11 = TextureSampleParameter2D(UVs=Multiply(
    A=ScalarParameter() {defaultValue=0.01, group=DesertSand01, parameterName=DesertSand01_DistortionScale, sortPriority=-500}
    B=LandscapeLayerCoords()
   )) {
   group=DesertSand01
   parameterName=DesertSand01_DistortionTexture
   samplerSource=SSM_Wrap_WorldGroupSettings
   texture=/Game/Weapons/WeaponSkins/Wraps/Textures/Conscience/T_Wrap_Conscience_Flow2
  }
  n12 = Subtract(A=@n11.RGB) {constB=0.5}
  n13 = ScalarParameter() {defaultValue=1.75, group=DesertSand01, parameterName=DesertSand01_DistortionAmount, sortPriority=-500}
  n14 = Add(
   A=ComponentMask(None=Multiply(A=@n12, B=@n13)) {g=true, r=true}
   B=Multiply(A=ScalarParameter() {defaultValue=0.1, group=DesertSand01, parameterName=DesertSand01_Scale, sortPriority=-500}, B=LandscapeLayerCoords())
  )
  n15 = TextureSample(UVs=@n14) {
   samplerSource=SSM_Wrap_WorldGroupSettings
   samplerType=SAMPLERTYPE_Normal
   texture=/Game/Environments/Helios/Terrain/Textures/T_Helios_WavySand_01_N
  }
  MaterialFunctionCall!(Normals=@n15.RGB, TransformToTangentSpace=NamedRerouteUsage(), Normal Strength) {materialFunction=/Game/Athena/Apollo/Environments/Landscape/Materials/Biome/CommonMaterials/MF_HQ_WorldAlignedNormal}
  n16 = FunctionInput(Preview=StaticBool()) {
   bUsePreviewValueAsDefault=true
   description="Default = False"
   inputName=UseTangentNormals
   inputType=FunctionInput_StaticBool
   sortPriority=15
  }
  NamedRerouteDeclaration(None=@n16) {name=UseTangentNormalsBool, nodeColor=<rgba 0.5 0 1 1>}
  NamedRerouteDeclaration(None=FunctionInput(Preview=Constant()) {bUsePreviewValueAsDefault=true, inputName=VariationMasks Mobile, inputType=FunctionInput_Vector4, sortPriority=40}) {name=VariationMasks Mobile, nodeColor=<rgba 1 0.8498 0.0704 1>}
  n17 = MaterialFunctionCall(Float4=NamedRerouteUsage()) {materialFunction=&BreakOutFloat4Components}
  NamedRerouteDeclaration(None=@n17.R) {name=OD_MD_Layer1, nodeColor=<rgba 0.5 0 1 1>}
  NamedRerouteDeclaration(None=@n17.G) {name=OD_MD_Layer2, nodeColor=<rgba 0.5 0 1 1>}
  n18 = SetMaterialAttributes!(MaterialAttributes)
  n19 = FunctionInput(Preview=@n18) {bUsePreviewValueAsDefault=true, inputName=Mobile Sand Material, inputType=FunctionInput_MaterialAttributes, sortPriority=11}
  NamedRerouteDeclaration(None=@n19) {name=Mobile Sand, nodeColor=<rgba 0 1 0.0156 1>}
  NamedRerouteDeclaration(None=@n17.B) {name=OD_MD_Layer3, nodeColor=<rgba 0.5 0 1 1>}
  NamedRerouteDeclaration(None=@n17.A) {name=OD_MD_Layer4, nodeColor=<rgba 0.5 0 1 1>}
  n20 = VectorParameter() {group=BiomeColors, parameterName=OD_MD_Color0, defaultValue=<rgba 0.45 0.1445 0.045 1>}
  n21 = VectorParameter() {group=BiomeColors, parameterName=OD_MD_Color1, defaultValue=<rgba 0.6 0.1927 0.06 1>}
  n22 = TextureSample(UVs=@n14) {samplerSource=SSM_Wrap_WorldGroupSettings, texture=/Game/Environments/Helios/Terrain/Textures/T_Helios_WavySand_01_D}
  n23 = OneMinus(None=@n22.A)
  n24 = ScalarParameter() {defaultValue=0.5, group=OpenDesert_MacroDetail, parameterName=OD_MD_LayerBlendContrast, sortPriority=501}
  n25 = MaterialFunctionCall(Transition Phase=NamedRerouteUsage(), Height Texture=@n23, Contrast=@n24) {materialFunction=&HeightLerp}
  n26 = NamedRerouteUsage()
  n27 = LinearInterpolate(
   A=LinearInterpolate(
    A=LinearInterpolate(A=@n20.RGB, B=@n21.RGB, Alpha=@n25.Alpha)
    B=VectorParameter() {group=BiomeColors, parameterName=OD_MD_Color2, defaultValue=<rgba 0.65 0.2823 0.1625 1>}.RGB
    Alpha=MaterialFunctionCall(Transition Phase=@n26, Height Texture=@n23, Contrast=@n24) {materialFunction=&HeightLerp}.Alpha
   )
   B=VectorParameter() {group=BiomeColors, parameterName=OD_MD_Color3, defaultValue=<rgba 0.8 0.275 0.104 1>}.RGB
   Alpha=MaterialFunctionCall(Transition Phase=NamedRerouteUsage(), Height Texture=@n23, Contrast=@n24) {materialFunction=&HeightLerp}.Alpha
  )
  n28 = VectorParameter() {group=BiomeColors, parameterName=OD_MD_Color4, defaultValue=<rgba 0.86 0.354 0.1892 1>}
  n29 = MaterialFunctionCall(Transition Phase=NamedRerouteUsage(), Height Texture=@n23, Contrast=@n24) {materialFunction=&HeightLerp}
  n30 = TextureSample(UVs=@n14) {
   samplerSource=SSM_Wrap_WorldGroupSettings
   samplerType=SAMPLERTYPE_Masks
   texture=/Game/Environments/Helios/Terrain/Textures/T_Helios_WavySand_01_S
  }
  n31 = VectorParameter() {group=DesertSand01, parameterName=DesertSand01_NormalMod, defaultValue=<rgba 1 1 1 0>}
  n32 = Transform(None=ComponentMask(None=Multiply(A=@n15.RGB, B=@n31.RGB)) {b=true, g=true, r=true})
  n33 = NamedRerouteUsage()
  n34 = SetMaterialAttributes!(
   MaterialAttributes
   BaseColor=Multiply(A=LinearInterpolate(A=@n27, B=@n28.RGB, Alpha=@n29.Alpha), B=@n22.RGB)
   Specular=Clamp(None=@n30.R) {clampMode=CMODE_ClampMax, maxDefault=0.6}
   Roughness=Clamp(
    None=@n30.B
    Min=ScalarParameter() {defaultValue=0.5, group=DesertSand01, parameterName=DesertSand01_RoughMin, sortPriority=-500}
    Max=ScalarParameter() {defaultValue=0.8, group=DesertSand01, parameterName=DesertSand01_RoughMax, sortPriority=-500}
   ) {maxDefault=0.5}
   Normal=Normalize(VectorInput=Add(A=@n32, B=@n33))
   OpacityMask=@n22.A
   Displacement=LinearInterpolate(B=@n22.A, Alpha=ScalarParameter() {defaultValue=1, group=Displacement, parameterName=DesertSand01_DisplacementAmt, sortPriority=0}) {constA=0.5}
  ) {attributeSetTypes [
    #1
    #2
    #3
    #4
    #5
    #6
   ]}
  NamedRerouteDeclaration(None=@n34) {name=LayerMaterial, nodeColor=<rgba 0 1 0 1>}
  NamedRerouteDeclaration(None=FunctionInput(Preview=Constant()) {bUsePreviewValueAsDefault=true, inputName=LayerWeight, inputType=FunctionInput_Scalar, sortPriority=15}) {name=LayerWeight, nodeColor=<rgba 0.5 0 1 1>}
  NamedRerouteDeclaration(None=@n6.LayerMask) {name=LayerMask, nodeColor=<rgba 0.7344 0 1 1>}
  NamedRerouteDeclaration(None=FunctionInput(Preview=@n18) {bUsePreviewValueAsDefault=true, inputName=Material, inputType=FunctionInput_MaterialAttributes}) {name=InputMaterial, nodeColor=<rgba 0 1 0 1>}
  NamedRerouteDeclaration(None=FunctionInput(Preview=Constant()) {bUsePreviewValueAsDefault=true, inputName=TerrainMasks_QuadA, inputType=FunctionInput_Vector4, sortPriority=20}) {name=TerrainMasks_QuadA, nodeColor=<rgba 0.5 0 1 1>}
  NamedRerouteDeclaration(None=FunctionInput(Preview=Constant()) {bUsePreviewValueAsDefault=true, inputName=TerrainMasks_QuadC, inputType=FunctionInput_Vector4, sortPriority=30}) {name=TerrainMasks_QuadC, nodeColor=<rgba 0.5 0 1 1>}
  NamedRerouteDeclaration(None=@n10.LayerMask) {name=MobileLayerMask, nodeColor=<rgba 0 0.9062 1 1>}
  NamedRerouteDeclaration(None=FunctionInput(Preview=Constant()) {bUsePreviewValueAsDefault=true, inputName=TerrainMasks_QuadD, inputType=FunctionInput_Vector4, sortPriority=35}) {name=TerrainMasks_QuadD, nodeColor=<rgba 0.5 0 1 1>}
  NamedRerouteDeclaration(None=FunctionInput(Preview=Constant()) {bUsePreviewValueAsDefault=true, inputName=TerrainMasks_QuadB, inputType=FunctionInput_Vector4, sortPriority=25}) {name=TerrainMasks_QuadB, nodeColor=<rgba 0.5 0 1 1>}
  NamedRerouteDeclaration(None=FunctionInput(Preview=@n18) {bUsePreviewValueAsDefault=true, inputName=MobileMaterial, inputType=FunctionInput_MaterialAttributes, sortPriority=10}) {name=MobileInputMaterial, nodeColor=<rgba 0 0.5 1 1>}
  NamedRerouteDeclaration(None=NamedRerouteUsage()) {name=MobileLayerMaterial, nodeColor=<rgba 0 0.5 1 1>}
  NamedRerouteDeclaration(None=StaticBool() {value=true}) {name=UseContrast, nodeColor=<rgba 0.5 0 1 1>}
  n35 = Constant3Vector()
  NamedRerouteDeclaration(None=FunctionInput(Preview=@n35) {bUsePreviewValueAsDefault=true, inputName=OD_MD_Normals, sortPriority=50}) {name=OD_MD_Normals, nodeColor=<rgba 0.5 0 1 1>}
  NamedRerouteDeclaration(None=ScalarParameter() {defaultValue=0.5, group=DesertSand01, parameterName=DesertSand01LayerBlendContrast, sortPriority=500}) {name=LayerBlendContrast, nodeColor=<rgba 0.5 0 1 1>}
  NamedRerouteDeclaration(None=FunctionInput(Preview=Constant4Vector().RGBA) {bUsePreviewValueAsDefault=true, inputName=OD_MD_SedimentLayers, inputType=FunctionInput_Vector4, sortPriority=60}) {name=OD_MD_SedimentLayers, nodeColor=<rgba 0.5 0 1 1>}
  NamedRerouteDeclaration(None=FunctionInput(Preview=@n35) {bUsePreviewValueAsDefault=true, inputName=OD_MD_UnmodifiedNormals, sortPriority=55}) {name=OD_MD_UnmodifiedNormals, nodeColor=<rgba 0.5 0 1 1>}
  NamedRerouteUsage()
  NamedRerouteUsage()
  NamedRerouteUsage()
  NamedRerouteUsage()
  NamedRerouteUsage()
  NamedRerouteUsage()
  n36 = ScalarParameter() {defaultValue=1, group=Displacement, parameterName=DesertSand01_Displacement_Power}
  Saturate(None=Multiply(A=Power(Base=@n22.A, Exp=@n36), B=ScalarParameter() {defaultValue=1, group=Displacement, parameterName=DesertSand01_Displacement_Mod}))
 ]}
