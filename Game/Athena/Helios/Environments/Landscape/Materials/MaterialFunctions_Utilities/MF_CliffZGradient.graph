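# Material function that tints landscape cliff base color with banded gradient texture T_Cliff_Grad.
# Axis-weighted world position (X counted twice, Y unused, plus Z), divided by a size parameter and offset by roughness, forms the gradient coordinate.
# That coordinate is rotated by a band angle before sampling.
# The biased sample multiplies BaseColor, blended by an opacity parameter and written back into the material attributes.
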MaterialFunction /Game/Athena/Helios/Environments/Landscape/Materials/MaterialFunctions_Utilities/MF_CliffZGradient
graph {nodes [
  n1 = GetMaterialAttributes(None=FunctionInput() {inputType=FunctionInput_MaterialAttributes}) {attributeGetTypes [#1, #2]}
  n2 = ScalarParameter() {group=Cliff, parameterName=CliffZGradient_BandAngle, sortPriority=510}
  n3 = ScalarParameter() {defaultValue=0.3, group=Cliff, parameterName=CliffZGradient_RoughnessMod, sortPriority=510}
  n4 = LinearInterpolate(A=Multiply(A=@n3) {constB=-1}, B=@n3, Alpha=@n1.Roughness)
  n5 = VectorParameter() {group=Cliff, parameterName=CliffZGradient_AxisIntensity, sortPriority=500, defaultValue=<rgba -0.1 -0.1 1 1>}
  n6 = MaterialFunctionCall(Float3=Multiply(A=WorldPosition().XYZ, B=@n5.RGB)) {materialFunction=&BreakOutFloat3Components}
  n7 = Divide(
   A=Add(A=Add(A=@n6.R, B=@n6.R), B=@n6.B)
   B=ScalarParameter() {defaultValue=10000, group=Cliff, parameterName=CliffZGradient_Size, sortPriority=510}
  )
  n8 = TextureSample(UVs=MaterialFunctionCall!(
    Angle=MaterialFunctionCall!(Degrees=@n2) {materialFunction=/Engine/Functions/Engine_MaterialFunctions01/Math/DegreesToRadians}.Radians
    UV/2DVector=Add(A=@n4, B=@n7)
   ) {materialFunction=/Game/Characters/Player/Male/Medium/Bodies/M_MED_Faux/Materials/MF_2DRotate}.Result) {samplerSource=SSM_Wrap_WorldGroupSettings, texture=/Game/Athena/Asteria/Environments/Landscape/Materials/MaterialFunctions/T_Cliff_Grad}
  n9 = Multiply(
   A=@n1.BaseColor
   B=Add(A=@n8.RGB, B=ScalarParameter() {defaultValue=0.2, group=Cliff, parameterName=CliffZGradient_PreBias, sortPriority=510})
  )
  n10 = ScalarParameter() {defaultValue=0.5, group=Cliff, parameterName=CliffZGradient_Opacity, sortPriority=510}
  FunctionOutput(None=SetMaterialAttributes!(MaterialAttributes=@n1.MaterialAttributes, BaseColor=LinearInterpolate(A=@n1.BaseColor, B=@n9, Alpha=@n10)) {attributeSetTypes [#1]})
 ]}
