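# Landscape material function that tints cliff BaseColor by biome, centered on a Chaparral Grassland tint with a separate moss color.
# Terrain mask quads B and D split into eight biome masks, whose inverted maxima define the Grassland mask.
# The output path desaturates color, multiplies it by tint parameters, blends moss and masks the result into BaseColor.
# Coast and Boreal tint lerps exist but feed nothing, and quads A and C are unconnected.
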MaterialFunction /Game/Athena/Helios/Environments/Landscape/Materials/MaterialFunctions_Utilities/MF_Cliff_BiomeColorTint
graph {
 properties {libraryCategoriesText [FN-Landscape]}
 nodes [
  FunctionInput() {bUsePreviewValueAsDefault=true, inputName=TerrainMasks_QuadA, inputType=FunctionInput_Vector4, sortPriority=10}
  FunctionInput() {bUsePreviewValueAsDefault=true, inputName=TerrainMasks_QuadC, inputType=FunctionInput_Vector4, sortPriority=30}
  FunctionOutput(None=SetMaterialAttributes!(MaterialAttributes=NamedRerouteUsage(), BaseColor=NamedRerouteUsage()) {attributeSetTypes [#1]})
  n1 = VectorParameter() {group=BiomeColors, parameterName=CliffColor_Chaparral_Coast, defaultValue=<rgba 1 1 1 1>}
  n2 = Multiply(A=NamedRerouteUsage(), B=@n1.RGB)
  n3 = LinearInterpolate(A=NamedRerouteUsage(), B=Saturate(None=Multiply(A=@n2, B=@n1.A)), Alpha=NamedRerouteUsage())
  n4 = NamedRerouteUsage()
  n5 = VectorParameter() {group=BiomeColors, parameterName=CliffColor_Boreal_Tundra, defaultValue=<rgba 1 1 1 1>}
  n6 = Saturate(None=Multiply(A=Multiply(A=@n4, B=@n5.RGB), B=@n5.A))
  n7 = NamedRerouteUsage()
  n8 = VectorParameter() {group=BiomeColors, parameterName=CliffColor_Boreal_Forest, defaultValue=<rgba 1 1 1 1>}
  n9 = Multiply(A=Multiply(A=NamedRerouteUsage(), B=@n8.RGB), B=@n8.A)
  n10 = VectorParameter() {group=BiomeColors, parameterName=CliffColor_Boreal_Mountain, defaultValue=<rgba 1 1 1 1>}
  n11 = Multiply(A=NamedRerouteUsage(), B=@n10.RGB)
  LinearInterpolate(
   A=LinearInterpolate(A=LinearInterpolate(A=@n3, B=@n6, Alpha=@n7), B=Saturate(None=@n9), Alpha=NamedRerouteUsage())
   B=Saturate(None=Multiply(A=@n11, B=@n10.A))
   Alpha=NamedRerouteUsage()
  )
  n12 = NamedRerouteUsage()
  n13 = NamedRerouteUsage()
  n14 = NamedRerouteUsage()
  Multiply(A=@n12, B=OneMinus(None=Max(A=@n13, B=@n14)))
  NamedRerouteDeclaration(None=FunctionInput() {bUsePreviewValueAsDefault=true, inputName=MossMask, inputType=FunctionInput_Scalar, sortPriority=100}) {name=MossMask, nodeColor=<rgba 0.5 0 1 1>}
  n15 = NamedRerouteUsage()
  n16 = Desaturation(None=NamedRerouteUsage(), Fraction=ScalarParameter() {defaultValue=0.5, group=BiomeColors, parameterName=Cliff_Chaparral_Grassland_Desat})
  n17 = VectorParameter() {group=BiomeColors, parameterName=CliffColor_Chaparral_Grassland, defaultValue=<rgba 0.5104 0.4672 0.4247 1>}
  n18 = Multiply(A=@n16, B=@n17.RGB)
  n19 = VectorParameter() {group=BiomeColors, parameterName=CliffColor_Chaparral_Grassland_Moss, defaultValue=<rgba 0.4067 0.4792 0.35 0>}
  n20 = LinearInterpolate(A=Saturate(None=Multiply(A=@n18, B=@n17.A)), B=Multiply(A=@n16, B=@n19.RGB), Alpha=NamedRerouteUsage())
  NamedRerouteDeclaration(None=ComponentMask(None=LinearInterpolate(A=@n15, B=@n20, Alpha=@n12)) {b=true, g=true, r=true}) {name=OutputBasecolor, nodeColor=<rgba 0.5 0 1 1>}
  n21 = GetMaterialAttributes(None=FunctionInput(Preview=SetMaterialAttributes!(MaterialAttributes)) {bUsePreviewValueAsDefault=true, inputName=Material, inputType=FunctionInput_MaterialAttributes}) {attributeGetTypes [#1]}
  NamedRerouteDeclaration(None=Desaturation(
    None=NamedRerouteDeclaration(None=@n21.BaseColor) {name=Input_BaseColor, nodeColor=<rgba 0.5 0 1 1>}
    Fraction=ScalarParameter() {defaultValue=0.5, group=BiomeColors, parameterName=Cliff_BiomeDesat}
   )) {name=BaseColorDesat, nodeColor=<rgba 0.5 0 1 1>}
  NamedRerouteDeclaration(None=@n21.MaterialAttributes) {name=InputMaterial, nodeColor=<rgba 0.5 0 1 1>}
  n22 = MaterialFunctionCall(Float4=FunctionInput() {bUsePreviewValueAsDefault=true, inputName=TerrainMasks_QuadB, inputType=FunctionInput_Vector4, sortPriority=20}) {materialFunction=&BreakOutFloat4Components}
  n23 = Max(
   A=Max(
    A=NamedRerouteDeclaration(None=@n22.R) {name=Biome_Chaparral_Coast, nodeColor=<rgba 0.5 0 1 1>}
    B=NamedRerouteDeclaration(None=@n22.G) {name=Biome_Boreal_Forest, nodeColor=<rgba 0.5 0 1 1>}
   )
   B=Max(
    A=NamedRerouteDeclaration(None=@n22.B) {name=Biome_Boreal_Tundra, nodeColor=<rgba 0.5 0 1 1>}
    B=NamedRerouteDeclaration(None=@n22.A) {name=Biome_Boreal_Mountain, nodeColor=<rgba 0.5 0 1 1>}
   )
  )
  n24 = MaterialFunctionCall(Float4=FunctionInput() {bUsePreviewValueAsDefault=true, inputName=TerrainMasks_QuadD, inputType=FunctionInput_Vector4, sortPriority=40}) {materialFunction=&BreakOutFloat4Components}
  n25 = Max(
   A=Max(
    A=NamedRerouteDeclaration(None=@n24.R) {name=Biome_Hades, nodeColor=<rgba 0.5 0 1 1>}
    B=NamedRerouteDeclaration(None=@n24.G) {name=Biome_Olympus, nodeColor=<rgba 0.5 0 1 1>}
   )
   B=Max(
    A=NamedRerouteDeclaration(None=@n24.B) {name=Biome_Arid_Desert, nodeColor=<rgba 0.5 0 1 1>}
    B=NamedRerouteDeclaration(None=@n24.A) {name=Biome_Open_Desert, nodeColor=<rgba 0.5 0 1 1>}
   )
  )
  NamedRerouteDeclaration(None=Multiply(A=OneMinus(None=@n23), B=OneMinus(None=@n25))) {name=Biome_Chaparral_Grassland, nodeColor=<rgba 0.5 0 1 1>}
 ]
}
