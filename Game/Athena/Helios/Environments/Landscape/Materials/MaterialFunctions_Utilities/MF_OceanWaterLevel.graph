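# MF_OceanWaterLevel is a landscape material function that derives height-based 0–1 gradient masks around the ocean water level.
# It adds an adjustable ocean level offset to world-position Z as a shared height value.
# SandWaterLevel adds a slight wet-sand height bias, then divides by a fixed falloff and saturates.
# CustomFalloff saturates the unbiased height divided by an input falloff amount, which affects only that output.
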MaterialFunction /Game/Athena/Helios/Environments/Landscape/Materials/MaterialFunctions_Utilities/MF_OceanWaterLevel
graph {nodes [
  Constant() {r=65}
  n1 = Add(
   A=WorldPosition().Z
   B=FunctionInput(Preview=Constant() {r=55}) {bUsePreviewValueAsDefault=true, inputName=OceanLevelOffset, inputType=FunctionInput_Scalar}
  )
  n2 = Add(A=@n1, B=Constant() {desc=slightly wet sand height offset, r=10})
  n3 = Constant() {r=125}
  FunctionOutput(None=Saturate(None=Divide(A=@n2, B=@n3))) {outputName=SandWaterLevel}
  n4 = FunctionInput(Preview=Constant() {r=125}) {
   bUsePreviewValueAsDefault=true
   description=Adjust falloff for CustomFalloff output only
   inputName=CustomFalloffAmount
   inputType=FunctionInput_Scalar
  }
  FunctionOutput(None=Saturate(None=Divide(A=@n1, B=@n4))) {outputName=CustomFalloff}
 ]}
