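# Material function that adds the Helios landscape cliff layer, outputting blended desktop and mobile materials plus layer and grass masks.
# Two world-aligned rock sets height-lerp by noise, or a static switch selects simplified sedimentary projection.
# The rock then gets moss desaturation, detail texture, ocean-level darkening and biome tinting.
# It height-blends over input terrain with slope-masked opacity and displacement, while mobile blends separate simplified cliff functions.
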MaterialFunction /Game/Athena/Helios/Environments/Landscape/Materials/LandscapeLayers/MF_Helios_Cliff
graph {nodes [
  Composite() {subgraphName=HighEndCliff}
  Constant() {r=-1.25}
  Constant() {r=1}
  n1 = NamedRerouteUsage()
  FunctionOutput(None=@n1) {outputName=CliffMaterialOnly, sortPriority=8}
  n2 = NamedRerouteUsage()
  n3 = GetMaterialAttributes(None=@n2) {attributeGetTypes [#1, #2]}
  n4 = NamedRerouteUsage()
  MaterialExpressionReroute_3 = Reroute(None=@n4)
  n5 = MaterialFunctionCall(Transition Phase=@MaterialExpressionReroute_3.SlopeMaskForInputHeightmap, Height Texture=@n3.Displacement) {materialFunction=&HeightLerp}
  n6 = LinearInterpolate(A=Constant() {r=0.5}, B=@n3.Displacement, Alpha=@n5.Lerp Alpha No Contrast)
  FunctionOutput(None=@n6) {outputName=DEBUG_Out, sortPriority=500}
  n7 = NamedRerouteUsage()
  n8 = NamedRerouteUsage()
  n9 = NamedRerouteUsage()
  n10 = GetMaterialAttributes(None=@n2) {attributeGetTypes [#1]}
  n11 = Multiply(
   A=@n10.OpacityMask
   B=MaterialFunctionCall(Transition Phase=@MaterialExpressionReroute_3.SlopeMaskForInputHeightmap, Height Texture=@n10.OpacityMask) {materialFunction=&HeightLerp}.Lerp Alpha No Contrast
  )
  n12 = SetMaterialAttributes!(MaterialAttributes=@n3.MaterialAttributes, OpacityMask=@n11, Displacement=@n6) {attributeSetTypes [#1, #2]}
  n13 = MaterialFunctionCall!(InputMaterial=@n12, LayerMaterial=@n1, LayerWeight=@n7, UseContrast=@n8, Contrast=@n9) {materialFunction=/Game/Athena/Helios/Environments/Landscape/Materials/MaterialFunctions_Utilities/MF_HeightBlendLandscapeLayers_OpacityMaskDisplaceCustom}
  FunctionOutput(None=@n13.LayerMask) {outputName=LayerMask, sortPriority=-1}
  n14 = MaterialFunctionCall!(
   InputMaterial=NamedRerouteUsage()
   LayerMaterial=NamedRerouteUsage()
   LayerWeight=NamedRerouteUsage()
   UseContrast=NamedRerouteUsage()
   Contrast=NamedRerouteUsage()
  ) {materialFunction=/Game/Athena/Helios/Environments/Landscape/Materials/MaterialFunctions_Utilities/MF_HeightBlendLandscapeLayers_OpacityMask}
  FunctionOutput(None=@n14.BlendedMaterial) {outputName=MobileBlendedMaterial, sortPriority=-100}
  n15 = NamedRerouteUsage()
  n16 = NamedRerouteUsage()
  n17 = Add(A=OneMinus(None=ShadingPathSwitch(Default=@n15, Mobile=@n16)), B=Constant() {r=-0.42})
  n18 = ComponentMask(None=VertexNormalWS()) {b=true}
  n19 = Saturate(None=Add(A=Saturate(None=@n17), B=Abs(None=@n18)))
  FunctionOutput(None=Ceil(None=OneMinus(None=@n19))) {outputName=GrassMask, sortPriority=5}
  FunctionOutput(None=@n13.BlendedMaterial) {outputName=BlendedMaterial, sortPriority=-300}
  n20 = NamedRerouteUsage()
  n21 = NamedRerouteUsage()
  n22 = TextureObject() {samplerType=SAMPLERTYPE_LinearGrayscale, texture=/Game/Global/Textures/Default/Noise/T_SwirlNoise_N_height}
  n23 = TextureSample(
   UVs=MaterialFunctionCall!(Scale=ScalarParameter() {defaultValue=-10000, group=Cliff, parameterName=Cliff_01and02_VariationNoise_Scale, sortPriority=98}) {materialFunction=/Engine/Functions/Engine_MaterialFunctions01/Coordinates/WorldPosition-XY}.Result
   Tex=@n22
  ) {
   samplerSource=SSM_Wrap_WorldGroupSettings
   samplerType=SAMPLERTYPE_LinearGrayscale
   texture=&T_SwirlNoise_N_height
  }
  n24 = MaterialFunctionCall!(
   Diffuse 1
   Diffuse 2
   Transistion Phase=@n23.R
   Height Texture 1=NamedRerouteUsage()
   Height Texture 2=NamedRerouteUsage()
   Contrast=Constant() {desc="Play with neg or pos ##", r=-5}
  ) {materialFunction=/Engine/Functions/Engine_MaterialFunctions02/Texturing/HeightLerpWithTwoHeightMaps}
  n25 = NamedRerouteDeclaration(None=@n24.Alpha) {name="Cliff_1&&2_HBlend", nodeColor=<rgba 0.8047 0 1 1>}
  n26 = StaticBoolParameter() {
   desc="usefull for sedimentary cliffs since those won`t benefit of having two texture sets and can better match the rock 2023 MF"
   group=Cliff
   parameterName=Cliff_ Use Simplified Biplanar
   sortPriority=-10
  }
  n27 = GetMaterialAttributes(None=StaticSwitch(True=NamedRerouteUsage(), False=BlendMaterialAttributes(A=@n20, B=@n21, Alpha=@n25), Value=@n26)) {attributeGetTypes [#3, #4, #1]}
  n28 = Multiply(
   A=@n27.BaseColor
   B=ComponentMask(None=VectorParameter() {group=Cliff, parameterName=Cliff_ColorTint, sortPriority=97, defaultValue=<rgba 1 1 1 1>}.RGB) {b=true, g=true, r=true}
  )
  n29 = Desaturation(None=@n28, Fraction=ScalarParameter() {defaultValue=0.8, group=Cliff, parameterName=Cliff_Moss_Mountains_DesaturationAmt})
  n30 = VectorParameter() {group=Cliff, parameterName=Spec_G Mask Tint Biom, channelNames {a=Blend to color, b="", g="", r=""}, defaultValue=<rgba 1 1 1 0>}
  n31 = LinearInterpolate(A=@n29, B=@n30, Alpha=@n30)
  n32 = LinearInterpolate(A=@n28, B=@n31, Alpha=NamedRerouteUsage())
  n33 = StaticBoolParameter() {group=Cliff, parameterName="Use L1 Spec_G for Tint && Spec - Biom Color Only"}
  n34 = VectorParameter() {group=Cliff, parameterName=Cliff_Moss_Mountains_ColorMod, defaultValue=<rgba 6 4.7066 4.38 0.16>}
  n35 = Saturate(None=Multiply(A=@n29, B=@n34.RGB))
  n36 = ComponentMask(None=NamedRerouteUsage()) {a=true}
  n37 = NamedRerouteUsage()
  n38 = LinearInterpolate(
   A=@n28
   B=StaticSwitch(
    True=StaticSwitch(True=@n31, False=@n32, Value=@n33)
    False=Multiply(A=@n35, B=@n34.A)
    Value=StaticBoolParameter() {group=Cliff, parameterName="Use L1 Spec_G for Tint && Spec - Biom Mask Use"}
   )
   Alpha=Multiply(A=NamedRerouteUsage(), B=Max(A=@n36, B=@n37) {constB=0})
  )
  n39 = Transform(None=@n27.Normal) {transformSourceType=TRANSFORMSOURCE_World, transformType=TRANSFORM_Tangent}
  n40 = NamedRerouteUsage()
  n41 = ScalarParameter() {defaultValue=1, group=Displacement, parameterName=Cliff_DisplacementAmt}
  n42 = GetMaterialAttributes(None=SetMaterialAttributes!(
    MaterialAttributes=@n27.MaterialAttributes
    BaseColor=StaticSwitchParameter(True=@n38, False=@n28) {group=Cliff, parameterName=Cliff_DesaturateMoss}
    Normal=StaticSwitch(True=@n39, False=@n27.Normal, Value=@n40)
    Displacement=LinearInterpolate(B=@n27.OpacityMask, Alpha=@n41) {constA=0.5}
   ) {attributeSetTypes [#3, #4, #2]}) {attributeGetTypes [#3, #5, #2]}
  n43 = TextureObjectParameter() {
   desc=This should match the Landscape Detail_Texture parameter in the SM cliff material
   group=Cliff
   parameterName=Cliff_DetailTexture
   texture=/Game/Environments/Apollo/Rocks/Natural_Rocks/Textures/T_RockDetails_02D
  }
  n44 = ScalarParameter() {
   defaultValue=700
   desc=This should match the Landscape Detail_Scale parameter in the SM cliff material
   group=Cliff
   parameterName=Cliff_DetailTexture_Scale
   sortPriority=98
  }
  n45 = MaterialFunctionCall!(
   Diffuse Texture Object=@n43
   Spec Texture Object
   TextureSize=@n44
   --------------------------
   Use seperate settings for Z projection
   Use seperate texture for Z projection
   Texture Size Z projection
   Z Plane Diffuse Texture Object
   Z Plane Spec Texture Object
   "--------------------------      "
   ProjectionTransitionContrast=Constant() {r=5}
   Use Texture Alpha Heightmap to Refine the Transistion
   HeightLerpTransitionContrast
   PickTheDiffuseHeightChannel
   TheDiffuseHeightChannel
   "--------------------------       "
   NormalTextureObject
   World Space Normals
   "--------------------------  "
   Export Float 4 Diffuse
   Export Float 4 Spec
   "--------------------------     "
   WorldPosition
   World Space Normal
  ) {materialFunction=/Engine/Functions/Engine_MaterialFunctions02/Texturing/WorldAlignedTextures_Complex}
  n46 = SquareRoot(None=@n45.XY Diffuse Texture)
  n47 = ScalarParameter() {
   defaultValue=1
   desc=This should match the Landscape Detail_Color parameter in the SM cliff material
   group=Cliff
   parameterName=Cliff_DetailTexture_Color
   sortPriority=2
  }
  n48 = Multiply(A=Multiply(A=@n42.BaseColor, B=LinearInterpolate(B=@n46, Alpha=@n47) {constA=1}), B=Constant() {r=1.5})
  n49 = Constant() {r=2.05}
  n50 = ComponentMask(None=@n45.XY Diffuse Texture) {r=true}
  n51 = Saturate(None=Add(A=Constant() {r=-1.1}, B=Multiply(A=@n49, B=@n50)))
  n52 = Saturate(None=LinearInterpolate(A=Constant() {r=7}, B=Constant() {r=-0.8}, Alpha=@n50))
  n53 = ScalarParameter() {
   defaultValue=1
   desc=This should match the Landscape Detail_Spec parameter in the SM cliff material
   group=Cliff
   parameterName=Cliff_DetailTexture_Spec
   sortPriority=2
  }
  n54 = LinearInterpolate(A=Constant() {r=-0.25}, B=Constant() {r=1.5}, Alpha=@n42.Specular)
  n55 = Multiply(
   A=LinearInterpolate(B=Multiply(A=@n51, B=@n52), Alpha=@n53) {constA=1}
   B=Clamp(None=Saturate(None=@n54), Max=Constant() {r=0.5}) {clampMode=CMODE_ClampMax}
  )
  n56 = GetMaterialAttributes(None=SetMaterialAttributes!(
    MaterialAttributes=@n42.MaterialAttributes
    BaseColor=Saturate(None=@n48)
    Specular=Saturate(None=@n55)
    Displacement=@n42.MaterialAttributes
   ) {attributeSetTypes [#3, #5, #2]}) {attributeGetTypes [#3, #6]}
  n57 = VectorParameter() {
   group=Cliff
   parameterName=Cliff_OceanLevelColorMod
   channelNames {a=RoughnessMod, b="", g="", r=""}
   defaultValue=<rgba 0.35 0.3086 0.2595 0.3>
  }
  n58 = ComponentMask(None=@n57) {b=true, g=true, r=true}
  n59 = Multiply(A=@n56.Roughness, B=@n57.RoughnessMod)
  n60 = NamedRerouteUsage()
  n61 = WorldPosition()
  n62 = ScalarParameter() {defaultValue=63, group=Cliff, parameterName=Cliff_OceanLevelOffset_Hard}
  n63 = Saturate(None=Divide(A=Add(A=@n61.Z, B=@n62), B=ScalarParameter() {defaultValue=10, group=Cliff, parameterName=Cliff_OceanLevelFalloff_Hard}))
  n64 = SetMaterialAttributes!(
   MaterialAttributes=@n56.MaterialAttributes
   BaseColor=LinearInterpolate(A=Multiply(A=@n56.BaseColor, B=@n58), B=@n56.BaseColor, Alpha=NamedRerouteUsage())
   Roughness=Max(A=LinearInterpolate(A=@n59, B=@n56.Roughness, Alpha=@n60), B=OneMinus(None=@n63))
  ) {attributeSetTypes [#3, #6]}
  MaterialFunctionCall!(MA=@n64, Biom Mask=NamedRerouteUsage(), Spec_G Mask=NamedRerouteUsage()) {materialFunction=/Game/Athena/Helios/Environments/Landscape/Materials/MaterialFunctions_Utilities/MF_CliffZGradient_2}
  MaterialFunctionCall(InputMaterial=@n12, LayerMaterial=@n1, LayerWeight=@n7, UseContrast=@n8, Contrast=@n9) {materialFunction=&MF_HeightBlendLandscapeLayers_OpacityMask}
  n65 = Abs(None=ComponentMask(None=VertexNormalWS()) {b=true})
  n66 = Saturate(None=Multiply(A=OneMinus(None=@n65), B=Constant() {r=3}))
  n67 = TextureSample(
   UVs=MaterialFunctionCall(Scale=Constant() {r=1200}) {materialFunction=&WorldPosition-XY}.Result
   Tex=@n22
  ) {
   samplerSource=SSM_Wrap_WorldGroupSettings
   samplerType=SAMPLERTYPE_LinearGrayscale
   texture=&T_SwirlNoise_N_height
  }
  n68 = MaterialFunctionCall(Transition Phase=@n66, Height Texture=OneMinus(None=@n67.A)) {materialFunction=&HeightLerp}
  Multiply(A=@n68.Lerp Alpha No Contrast)
  Multiply(A=@n29)
  Multiply(A=@n3.Displacement, B=@n5.Lerp Alpha No Contrast)
  Multiply(A=@n46, B=@n42.Displacement)
  n69 = TextureObjectParameter() {group=Cliff, parameterName=Cliff_02_D, sortPriority=200, texture=/Game/Environments/Asteria/Terrain/Textures/T_Asteria_RockCliff_03_D}
  n70 = TextureObjectParameter() {
   group=Cliff
   parameterName=Cliff_02_S
   samplerType=SAMPLERTYPE_Masks
   sortPriority=205
   texture=/Game/Environments/Asteria/Terrain/Textures/T_Asteria_RockCliff_03_S
  }
  n71 = TextureObjectParameter() {
   group=Cliff
   parameterName=Cliff_02_N
   samplerType=SAMPLERTYPE_Normal
   sortPriority=210
   texture=/Game/Environments/Asteria/Terrain/Textures/T_Asteria_RockCliff_03_N
  }
  n72 = MaterialFunctionCall(
   Diffuse Texture Object=@n69
   Spec Texture Object=@n70
   TextureSize=ComponentMask(None=VectorParameter() {group=Cliff, parameterName=Cliff_02_Scale, sortPriority=215, defaultValue=<rgba 3000 3000 3000 1>}.RGB) {b=true, g=true, r=true}
   ProjectionTransitionContrast=ScalarParameter() {defaultValue=0.5, desc=XY Projection Transition Contrast, group=Cliff, parameterName=Cliff_02-PTC, sortPriority=220}
   Use Texture Alpha Heightmap to Refine the Transistion=StaticBool() {value=true}
   HeightLerpTransitionContrast=ScalarParameter() {defaultValue=0.5, desc=XY HeightLerp Transition Contrast, group=Cliff, parameterName=Cliff_02-HLTC, sortPriority=225}
   NormalTextureObject=@n71
   World Space Normals=StaticBool() {value=true}
   Export Float 4 Diffuse=StaticBool() {value=true}
   WorldPosition=WorldPosition().XYZ
  ) {materialFunction=&WorldAlignedTextures_Complex}
  n73 = SetMaterialAttributes!(
   MaterialAttributes
   BaseColor=ComponentMask(None=@n72.XY Diffuse Texture) {b=true, g=true, r=true}
   Specular=ComponentMask(None=@n72.XY Spec) {r=true}
   Roughness=ComponentMask(None=@n72.XY Spec) {b=true}
   Normal=NamedRerouteDeclaration(None=@n72.Normal XY Texture) {name=Cliff_02_Normal, nodeColor=<rgba 0.5 0 1 1>}
   OpacityMask=NamedRerouteDeclaration(None=ComponentMask(None=@n72.XY Diffuse Texture) {a=true}) {name=Cliff_02_Height, nodeColor=<rgba 0.5 0 1 1>}
  ) {attributeSetTypes [
    #3
    #5
    #6
    #4
    #1
   ]}
  NamedRerouteDeclaration(None=@n73) {name=Cliff_02_Material, nodeColor=<rgba 0.5 0 1 1>}
  NamedRerouteDeclaration(None=FunctionInput(Preview=Constant()) {bUsePreviewValueAsDefault=true, inputName=VariationMasks Mobile, inputType=FunctionInput_Vector4, sortPriority=40}) {name=VariationMasks Mobile, nodeColor=<rgba 1 0.8498 0.0704 1>}
  n74 = ScalarParameter() {defaultValue=-1.25, group=Cliff, parameterName=Cliff_InnerSlopeMaskA, sortPriority=98}
  n75 = ScalarParameter() {defaultValue=1, group=Cliff, parameterName=Cliff_InnerSlopeMaskB, sortPriority=98}
  NamedRerouteDeclaration(None=Saturate(None=LinearInterpolate(A=@n74, B=@n75, Alpha=@n65))) {name=SlopeMaskForInputHeightmap, nodeColor=<rgba 0.8516 1 0 1>}
  NamedRerouteDeclaration(None=@n13.LayerMaskNoContrast) {name=LayerMaskNoContrast, nodeColor=<rgba 0 1 0.6544 1>}
  NamedRerouteDeclaration(None=@n14.LayerMaskNoContrast) {name=MobileLayerMaskNoContrast, nodeColor=<rgba 1 0 0.3047 1>}
  NamedRerouteDeclaration(None=FunctionInput(Preview=Constant()) {bUsePreviewValueAsDefault=true, inputName=MossDesaturateMask, inputType=FunctionInput_Scalar, sortPriority=45}) {name=MossDesaturateMask, nodeColor=<rgba 1 0.8498 0.0704 1>}
  n76 = FunctionInput(Preview=Constant()) {bUsePreviewValueAsDefault=true, inputName=TerrainMasks_QuadA, inputType=FunctionInput_Vector4, sortPriority=20}
  NamedRerouteDeclaration(None=ComponentMask(None=NamedRerouteDeclaration(None=@n76) {name=TerrainMasks_QuadA, nodeColor=<rgba 0.5 0 1 1>}) {g=true}) {name=OasisCliffMask, nodeColor=<rgba 0.5 0 1 1>}
  n77 = VectorParameter() {
   group=Cliff
   parameterName="Sediment_a | Pos && Scale"
   sortPriority=306
   channelNames {a=Scale, b=Z, g=Y, r=X}
   defaultValue=<rgba 0 0 0 2500>
  }
  n78 = VectorParameter() {
   group=Cliff
   parameterName="Sediment_a | Wave Dist Pos && Freq"
   sortPriority=307
   channelNames {a=Freq, b=Z, g=Y, r=X}
   defaultValue=<rgba 0 0 0 10000>
  }
  n79 = MaterialFunctionCall!(
   WorldPosition
   Sediment L2_WS | Position Offset_1=@n77
   Sediment L2_WS | Scale=@n77
   Use Wavy Distortion=StaticBoolParameter() {group=Cliff, parameterName=Sediment_a | Use Wave Distortion, sortPriority=306}
   Wave Distortion Intensity=ScalarParameter() {defaultValue=2500, group=Cliff, parameterName=Sediment_a | Wave Dist Intensity, sortPriority=307}
   Sediment L2_WS | Wave DistPos Offset=@n78
   Wave Distortion Freq=@n78
  ) {materialFunction=/Game/Global/Base4Layers_2022/MaterialFunctions/MF_Cliffs_SedimentWS_UVs}
  n80 = TextureObjectParameter() {group=Cliff, parameterName=Cliff_03_D, sortPriority=300, texture=/Game/Environments/Asteria/Terrain/Textures/T_Asteria_RockCliff_04_D}
  n81 = TextureObjectParameter() {
   group=Cliff
   parameterName=Cliff_03_S
   samplerType=SAMPLERTYPE_Masks
   sortPriority=300
   texture=/Game/Environments/Asteria/Terrain/Textures/T_Asteria_RockCliff_04_S
  }
  n82 = TextureObjectParameter() {
   group=Cliff
   parameterName=Cliff_03_N
   samplerType=SAMPLERTYPE_Normal
   sortPriority=300
   texture=&T_Asteria_RockCliff_03_N
  }
  n83 = ScalarParameter() {defaultValue=0.2, desc=XY Projection Transition Contrast, group=Cliff, parameterName=Sediment_a | PTC, sortPriority=306}
  n84 = MaterialFunctionCall!(UV_XZ=@n79.XZ, UV_YZ=@n79.XZ, T_Diffuse=@n80, T_Specular=@n81, T_Normal=@n82, PTC=@n83) {materialFunction=/Game/Global/Base4Layers_2022/MaterialFunctions/MF_Cliffs_SedimentWS}
  n85 = GetMaterialAttributes(None=@n84.SetMaterialAttributes) {attributeGetTypes [#7]}
  n86 = NamedRerouteUsage()
  NamedRerouteDeclaration(None=MaterialFunctionCall!(
    Sediment MA=@n85.MaterialAttributes
    IsLandscape=StaticBool() {value=true}
    Sediment Normal Intensity=ScalarParameter() {defaultValue=1, group=Cliff, parameterName=Sediment_b | L2 Cliff Normal Intensity, sortPriority=307}
    Cliffs MA=SetMaterialAttributes!(MaterialAttributes=@n85.MaterialAttributes, Opacity, Normal=@n86) {attributeSetTypes [#8, #4]}
    Sediment Wave Distiortion=@n79.Sediment Wave Distortion
    Sediment Normal B Mask=@n84.Sediment Normal B Mask
    Biom Mask=NamedRerouteUsage()
   ) {materialFunction=/Game/Athena/Helios/Environments/Landscape/Materials/MaterialFunctions_Utilities/MF_CliffSediment_Blend}.MA) {name=MA_Simple Biplanar, nodeColor=<rgba 0.0575 0 0.5711 1>}
  NamedRerouteDeclaration(None=@n85.Metallic) {name=Moss Mask03, nodeColor=<rgba 0.0575 0 0.5711 1>}
  n87 = TextureObjectParameter() {group=Cliff, parameterName=Cliff_01_D, sortPriority=100, texture=&T_Asteria_RockCliff_04_D}
  n88 = TextureObjectParameter() {
   group=Cliff
   parameterName=Cliff_01_S
   samplerType=SAMPLERTYPE_Masks
   sortPriority=105
   texture=&T_Asteria_RockCliff_04_S
  }
  n89 = VectorParameter() {group=Cliff, parameterName=Cliff_01_Scale, sortPriority=115, defaultValue=<rgba 4000 4000 4000 1>}
  n90 = TextureObjectParameter() {
   group=Cliff
   parameterName=Cliff_01_N
   samplerType=SAMPLERTYPE_Normal
   sortPriority=110
   texture=/Game/Environments/Asteria/Terrain/Textures/T_Asteria_RockCliff_04_N
  }
  n91 = MaterialFunctionCall(
   Diffuse Texture Object=@n87
   Spec Texture Object=@n88
   TextureSize=NamedRerouteDeclaration(None=ComponentMask(None=@n89.RGB) {b=true, g=true, r=true}) {name=Cliff_01_Scale, nodeColor=<rgba 0.5 0 1 1>}
   ProjectionTransitionContrast=ScalarParameter() {defaultValue=0.5, desc=XY Projection Transition Contrast, group=Cliff, parameterName=Cliff_01-PTC, sortPriority=120}
   Use Texture Alpha Heightmap to Refine the Transistion=StaticBool() {value=true}
   HeightLerpTransitionContrast=ScalarParameter() {defaultValue=0.5, desc=XY HeightLerp Transition Contrast, group=Cliff, parameterName=Cliff_01-HLTC, sortPriority=125}
   NormalTextureObject=@n90
   World Space Normals=StaticBool() {value=true}
   Export Float 4 Diffuse=StaticBool() {value=true}
   WorldPosition=WorldPosition().XYZ
  ) {materialFunction=&WorldAlignedTextures_Complex}
  NamedRerouteDeclaration(None=ComponentMask(None=@n91.XY Spec) {g=true}) {name=MossMask01, nodeColor=<rgba 0.5 0 1 1>}
  n92 = SetMaterialAttributes!(MaterialAttributes)
  NamedRerouteDeclaration(None=FunctionInput(Preview=@n92) {bUsePreviewValueAsDefault=true, inputName=Material, inputType=FunctionInput_MaterialAttributes}) {name=InputMaterial, nodeColor=<rgba 0 1 0 1>}
  n93 = MaterialFunctionCall!(In=@n64) {materialFunction=/Game/Athena/Helios/Environments/Landscape/Materials/MaterialFunctions_Utilities/MF_CliffZGradient}
  n94 = SetMaterialAttributes!(
   MaterialAttributes=MaterialFunctionCall!(
    Material=StaticSwitchParameter(True=@n93.Result, False=@n64) {desc=Uses Legacy CliffZGradient, group=Cliff, parameterName=Use CliffZGradient, sortPriority=510}
    TerrainMasks_QuadA=NamedRerouteUsage()
    TerrainMasks_QuadB=NamedRerouteUsage()
    TerrainMasks_QuadC=NamedRerouteUsage()
    TerrainMasks_QuadD=NamedRerouteUsage()
    MossMask=NamedRerouteUsage()
   ) {materialFunction=/Game/Athena/Helios/Environments/Landscape/Materials/MaterialFunctions_Utilities/MF_Cliff_BiomeColorTint}.Result
   BaseColor
   OpacityMask=NamedRerouteUsage()
  ) {desc=HighEnd Material, attributeSetTypes [#3, #1]}
  MaterialExpressionReroute_0 = Reroute(None=@n94)
  NamedRerouteDeclaration(None=@MaterialExpressionReroute_0.Material) {name=LayerMaterial, nodeColor=<rgba 0 1 0 1>}
  NamedRerouteDeclaration(None=FunctionInput(Preview=Constant()) {bUsePreviewValueAsDefault=true, inputName=LayerWeight, inputType=FunctionInput_Scalar, sortPriority=15}) {name=LayerWeight, nodeColor=<rgba 0.5 0 1 1>}
  NamedRerouteDeclaration(None=ComponentMask(None=@n72.XY Spec) {g=true}) {name=MossMask02, nodeColor=<rgba 0.502 0 1 1>}
  NamedRerouteDeclaration(None=FunctionInput(Preview=Constant()) {bUsePreviewValueAsDefault=true, inputName=TerrainMasks_QuadC, inputType=FunctionInput_Vector4, sortPriority=30}) {name=TerrainMasks_QuadC, nodeColor=<rgba 0.5 0 1 1>}
  NamedRerouteDeclaration(None=FunctionInput(Preview=Constant()) {bUsePreviewValueAsDefault=true, inputName=TerrainMasks_QuadD, inputType=FunctionInput_Vector4, sortPriority=35}) {name=TerrainMasks_QuadD, nodeColor=<rgba 0.5 0 1 1>}
  NamedRerouteDeclaration(None=FunctionInput(Preview=Constant()) {bUsePreviewValueAsDefault=true, inputName=TerrainMasks_QuadB, inputType=FunctionInput_Vector4, sortPriority=25}) {name=TerrainMasks_QuadB, nodeColor=<rgba 0.5 0 1 1>}
  n95 = SetMaterialAttributes!(
   MaterialAttributes
   BaseColor=ComponentMask(None=@n91.XY Diffuse Texture) {b=true, g=true, r=true}
   Specular=ComponentMask(None=@n91.XY Spec) {r=true}
   Roughness=ComponentMask(None=@n91.XY Spec) {b=true}
   Normal=NamedRerouteDeclaration(None=@n91.Normal XY Texture) {name=Cliff_01_Normal, nodeColor=<rgba 0.5 0 1 1>}
   OpacityMask=NamedRerouteDeclaration(None=ComponentMask(None=@n91.XY Diffuse Texture) {a=true}) {name=Cliff_01_Height, nodeColor=<rgba 0.5 0 1 1>}
   Opacity
  ) {attributeSetTypes [
    #3
    #5
    #6
    #4
    #1
    #8
   ]}
  NamedRerouteDeclaration(None=@n95) {name=Cliff_01_Material, nodeColor=<rgba 0.5 0 1 1>}
  NamedRerouteDeclaration(None=FunctionInput(Preview=@n92) {bUsePreviewValueAsDefault=true, inputName=MobileMaterial, inputType=FunctionInput_MaterialAttributes, sortPriority=10}) {name=MobileInputMaterial, nodeColor=<rgba 0 0.5 1 1>}
  n96 = NamedRerouteUsage()
  n97 = NamedRerouteUsage()
  n98 = NamedRerouteUsage()
  n99 = NamedRerouteUsage()
  n100 = NamedRerouteUsage()
  n101 = NamedRerouteUsage()
  NamedRerouteDeclaration(None=StaticSwitch(
    True=MaterialFunctionCall!(OasisMask=@n96, VariationMasks Mobile=@n97, IsStaticMesh) {materialFunction=/Game/Athena/Helios/Environments/Landscape/Materials/QualityLevelLayers/MF_Helios_CliffSedimentary_NS}.Result
    False=MaterialFunctionCall!(VariationMasks Mobile=@n97, TerrainMasks_QuadA=@n98, TerrainMasks_QuadB=@n99, TerrainMasks_QuadC=@n100, TerrainMasks_QuadD=@n101) {materialFunction=/Game/Athena/Helios/Environments/Landscape/Materials/QualityLevelLayers/MF_Helios_Cliff_NS}.Result
    Value=StaticBoolParameter() {parameterName=Mobile UseSedimentaryCliffs}
   )) {name=MobileLayerMaterial, nodeColor=<rgba 0 0.5 1 1>}
  NamedRerouteDeclaration(None=StaticBoolParameter() {defaultValue=true, group=Cliff, parameterName=UseCliff_LayerBlendContrast, sortPriority=-150}) {name=UseContrast, nodeColor=<rgba 0.5 0 1 1>}
  NamedRerouteDeclaration(None=Max(A=@n66, B=@n68.Lerp Alpha No Contrast) {desc="For BLEND MATERIAL, smallrockedge-mat should be in B"}) {name=Cliff_HeightMap, nodeColor=<rgba 0.8516 1 0 1>}
  NamedRerouteDeclaration(None=ScalarParameter() {defaultValue=0.5, group=Cliff, parameterName=Cliff_LayerBlendContrast, sortPriority=-100}) {name=LayerBlendContrast, nodeColor=<rgba 0.5 0 1 1>}
  n102 = FunctionInput(Preview=StaticBool()) {
   bUsePreviewValueAsDefault=true
   description="Default = False"
   inputName=UseTangentNormals
   inputType=FunctionInput_StaticBool
   sortPriority=15
  }
  NamedRerouteDeclaration(None=@n102) {name=UseTangentNormalsBool, nodeColor=<rgba 0.5 0 1 1>}
  n103 = ScalarParameter() {defaultValue=200, group=Cliff, parameterName=Cliff_OceanLevelFalloff}
  n104 = ScalarParameter() {defaultValue=52, group=Cliff, parameterName=Cliff_OceanLevelOffset}
  NamedRerouteDeclaration(None=MaterialFunctionCall!(CustomFalloffAmount=@n103, OceanLevelOffset=@n104) {materialFunction=/Game/Athena/Helios/Environments/Landscape/Materials/MaterialFunctions_Utilities/MF_OceanWaterLevel}.CustomFalloff) {name=WaterLevelMask, nodeColor=<rgba 0.5234 0 1 1>}
  n105 = NamedRerouteUsage()
  n106 = NamedRerouteUsage()
  NamedRerouteDeclaration(None=StaticSwitch(True=NamedRerouteUsage(), False=LinearInterpolate(A=@n105, B=@n106, Alpha=@n24.Alpha), Value=@n26)) {name=DesaturateMask, nodeColor=<rgba 0.5 0 1 1>}
  PinBase()
  PinBase!(Material=@n94, AlphaNoContrast, SlopeMaskForInputHeightmap=@n4) {reroutePins [
    Material {expression=&self:MaterialExpressionReroute_0}
    AlphaNoContrast {expression=&self:MaterialExpressionReroute_1}
    SlopeMaskForInputHeightmap {expression=&self:MaterialExpressionReroute_3}
   ]}
  MaterialExpressionReroute_1 = Reroute()
  Saturate(None=Divide(A=Add(A=@n61.Z, B=@n104), B=@n103))
  SetMaterialAttributes!(MaterialAttributes=@n10.MaterialAttributes, OpacityMask=@n11) {attributeSetTypes [#1]}
 ]}
