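# Landscape layer material function that blends Sand over the incoming Helios terrain material.
# Terrain masks drive separate default and mobile (NS) sand sub-functions, each height-blended by layer weight and a contrast parameter, with the default blend using a displacement-aware variant.
# Sand layer switches pass the input through when unused.
# Shading-path switches select the layer mask and a static-mesh material, whose mobile branch derives normals.
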MaterialFunction /Game/Athena/Helios/Environments/Landscape/Materials/LandscapeLayers/MF_Helios_Sand
graph {nodes [
  n1 = NamedRerouteUsage()
  n2 = NamedRerouteUsage()
  n3 = NamedRerouteUsage()
  n4 = NamedRerouteUsage()
  n5 = NamedRerouteUsage()
  n6 = MaterialFunctionCall!(InputMaterial=@n1, LayerMaterial=@n2, LayerWeight=@n3, UseContrast=@n4, Contrast=@n5) {materialFunction=/Game/Athena/Helios/Environments/Landscape/Materials/MaterialFunctions_Utilities/MF_HeightBlendLandscapeLayers_OpacityMaskDisplaceCustom}
  FunctionOutput(None=LandscapeLayerSwitch!(LayerUsed=@n6.BlendedMaterial, LayerNotUsed=@n1) {parameterName=Sand}) {outputName=BlendedMaterial, sortPriority=-300}
  FunctionOutput(None=ShadingPathSwitch(Default=NamedRerouteUsage(), Mobile=NamedRerouteUsage())) {outputName=LayerMask, sortPriority=-1}
  FunctionOutput(None=NamedRerouteUsage()) {outputName=MobileMaterial, sortPriority=-100}
  n7 = NamedRerouteUsage()
  n8 = MaterialFunctionCall!(
   InputMaterial=@n7
   LayerMaterial=NamedRerouteUsage()
   LayerWeight=NamedRerouteUsage()
   UseContrast=NamedRerouteUsage()
   Contrast=NamedRerouteUsage()
  ) {materialFunction=/Game/Athena/Helios/Environments/Landscape/Materials/MaterialFunctions_Utilities/MF_HeightBlendLandscapeLayers_OpacityMask}
  FunctionOutput(None=LandscapeLayerSwitch(LayerUsed=@n8.BlendedMaterial, LayerNotUsed=@n7) {parameterName=Sand}) {outputName=MobileBlendedMaterial, sortPriority=-100}
  n9 = NamedRerouteUsage()
  n10 = StaticBool()
  FunctionOutput(None=ShadingPathSwitch(
    Default=NamedRerouteUsage()
    Mobile=MaterialFunctionCall!(MobileMaterial=@n9, TransformTangents=@n10) {materialFunction=/Game/Athena/Helios/Environments/Landscape/Materials/MaterialFunctions_Utilities/MF_MobileDeriveNormals}.ResultMaterial
   )) {outputName=StaticMeshMaterial, sortPriority=10}
  MaterialFunctionCall(InputMaterial=@n1, LayerMaterial=@n2, LayerWeight=@n3, UseContrast=@n4, Contrast=@n5) {materialFunction=&MF_HeightBlendLandscapeLayers_OpacityMask}
  n11 = FunctionInput(Preview=StaticBool()) {
   bUsePreviewValueAsDefault=true
   description="Default = False"
   inputName=UseTangentNormals
   inputType=FunctionInput_StaticBool
   sortPriority=15
  }
  NamedRerouteDeclaration(None=@n11) {name=UseTangentNormalsBool, nodeColor=<rgba 0.5 0 1 1>}
  NamedRerouteDeclaration(None=FunctionInput(Preview=Constant()) {bUsePreviewValueAsDefault=true, inputName=VariationMasks Mobile, inputType=FunctionInput_Vector4, sortPriority=40}) {name=VariationMasks Mobile, nodeColor=<rgba 1 0.8498 0.0704 1>}
  NamedRerouteDeclaration(None=@n6.LayerMask) {name=LayerMask, nodeColor=<rgba 1 0 0.6094 1>}
  NamedRerouteDeclaration(None=MaterialFunctionCall!(
    TerrainMasks_QuadA=NamedRerouteUsage()
    TerrainMasks_QuadB=NamedRerouteUsage()
    TerrainMasks_QuadC=NamedRerouteUsage()
    TerrainMasks_QuadD=NamedRerouteUsage()
    UseTangentNormals=NamedRerouteUsage()
    OD_MD_Normals=NamedRerouteUsage()
    OD_MD_SedimentLayers=NamedRerouteUsage()
    UseOceanLevel
   ) {materialFunction=/Game/Athena/Helios/Environments/Landscape/Materials/QualityLevelLayers/MF_Helios_Sand_Default}.Material) {name=LayerMaterial, nodeColor=<rgba 0 1 0 1>}
  NamedRerouteDeclaration(None=FunctionInput(Preview=Constant()) {bUsePreviewValueAsDefault=true, inputName=LayerWeight, inputType=FunctionInput_Scalar, sortPriority=15}) {name=LayerWeight, nodeColor=<rgba 0.5 0 1 1>}
  NamedRerouteDeclaration(None=@n8.LayerMask) {name=MobileLayerMask, nodeColor=<rgba 0.7812 1 0 1>}
  n12 = SetMaterialAttributes!(MaterialAttributes)
  NamedRerouteDeclaration(None=FunctionInput(Preview=@n12) {bUsePreviewValueAsDefault=true, inputName=Material, inputType=FunctionInput_MaterialAttributes}) {name=InputMaterial, nodeColor=<rgba 0 1 0 1>}
  NamedRerouteDeclaration(None=FunctionInput(Preview=Constant()) {bUsePreviewValueAsDefault=true, inputName=TerrainMasks_QuadA, inputType=FunctionInput_Vector4, sortPriority=20}) {name=TerrainMasks_QuadA, nodeColor=<rgba 0.5 0 1 1>}
  NamedRerouteDeclaration(None=FunctionInput(Preview=Constant()) {bUsePreviewValueAsDefault=true, inputName=TerrainMasks_QuadC, inputType=FunctionInput_Vector4, sortPriority=30}) {name=TerrainMasks_QuadC, nodeColor=<rgba 0.5 0 1 1>}
  NamedRerouteDeclaration(None=FunctionInput(Preview=Constant3Vector()) {bUsePreviewValueAsDefault=true, inputName=OD_MD_Normals, sortPriority=50}) {name=OD_MD_Normals, nodeColor=<rgba 0.5 0 1 1>}
  NamedRerouteDeclaration(None=FunctionInput(Preview=Constant()) {bUsePreviewValueAsDefault=true, inputName=TerrainMasks_QuadD, inputType=FunctionInput_Vector4, sortPriority=35}) {name=TerrainMasks_QuadD, nodeColor=<rgba 0.5 0 1 1>}
  NamedRerouteDeclaration(None=FunctionInput(Preview=Constant()) {bUsePreviewValueAsDefault=true, inputName=TerrainMasks_QuadB, inputType=FunctionInput_Vector4, sortPriority=25}) {name=TerrainMasks_QuadB, nodeColor=<rgba 0.5 0 1 1>}
  NamedRerouteDeclaration(None=FunctionInput(Preview=@n12) {bUsePreviewValueAsDefault=true, inputName=MobileMaterial, inputType=FunctionInput_MaterialAttributes, sortPriority=10}) {name=MobileInputMaterial, nodeColor=<rgba 0 0.5 1 1>}
  NamedRerouteDeclaration(None=MaterialFunctionCall!(
    TerrainMasks_QuadA=NamedRerouteUsage()
    TerrainMasks_QuadB=NamedRerouteUsage()
    TerrainMasks_QuadC=NamedRerouteUsage()
    TerrainMasks_QuadD=NamedRerouteUsage()
    VariationMasks Mobile=NamedRerouteUsage()
   ) {materialFunction=/Game/Athena/Helios/Environments/Landscape/Materials/QualityLevelLayers/MF_Helios_Sand_NS}.Result) {name=MobileLayerMaterial, nodeColor=<rgba 0 0.5 1 1>}
  NamedRerouteDeclaration(None=StaticBool()) {name=UseContrast, nodeColor=<rgba 0.5 0 1 1>}
  NamedRerouteDeclaration(None=FunctionInput(Preview=Constant4Vector().RGBA) {bUsePreviewValueAsDefault=true, inputName=OD_MD_SedimentLayers, inputType=FunctionInput_Vector4, sortPriority=60}) {name=OD_MD_SedimentLayers, nodeColor=<rgba 0.5 0 1 1>}
  NamedRerouteDeclaration(None=ScalarParameter() {group=Sand, parameterName=SandLayerBlendContrast}) {name=LayerBlendContrast, nodeColor=<rgba 0.5 0 1 1>}
 ]}
