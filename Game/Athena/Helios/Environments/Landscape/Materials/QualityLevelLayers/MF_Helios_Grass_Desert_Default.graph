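# Helios landscape material function for the Desert Grass layer, blending chaparral coast ground with arid desert grass.
# Both texture sets use scalable world-space UVs and world-aligned normals, with tint and roughness adjustment on the coast ground.
# A height-blend function combines them, alongside world-space noise masks shaped by Beer's Law.
# It outputs the blend, also as StaticMeshMaterial, plus noise and weighted layer masks; the separate static-mesh attribute node is unconnected.
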MaterialFunction /Game/Athena/Helios/Environments/Landscape/Materials/QualityLevelLayers/MF_Helios_Grass_Desert_Default
graph {nodes [
  n1 = VectorParameter() {group=Grass_Desert, parameterName=Grass_Desert_Chap_ColorMod, defaultValue=<rgba 0.6 0.4446 0.3 1>}
  n2 = MaterialFunctionCall!(Scale=ScalarParameter() {defaultValue=900, group=Grass_Desert, parameterName=Grass_Desert_Chap_WorldScale}) {materialFunction=/Game/Athena/Apollo/Environments/Landscape/Materials/Biome/CommonMaterials/MF_SimpleWorldUV}
  n3 = TextureSample(UVs=@n2.Result, Tex=TextureObject() {texture=/Game/Environments/Helios/Terrain/Textures/T_Helios_Chap_CoastGround_01_D}) {samplerSource=SSM_Wrap_WorldGroupSettings, texture=&T_Helios_Chap_CoastGround_01_D}
  n4 = TextureSample(
   UVs=@n2.Result
   Tex=TextureObject() {samplerType=SAMPLERTYPE_Masks, texture=/Game/Environments/Helios/Terrain/Textures/T_Helios_Chap_CoastGround_01_S}
  ) {
   samplerSource=SSM_Wrap_WorldGroupSettings
   samplerType=SAMPLERTYPE_Masks
   texture=&T_Helios_Chap_CoastGround_01_S
  }
  n5 = MaterialFunctionCall(Float3=@n4.RGB) {materialFunction=&BreakOutFloat3Components}
  n6 = NamedRerouteDeclaration(None=@n5.B) {name=Grass_ChaparralCoast_Roughness, nodeColor=<rgba 0.5 0 1 1>}
  n7 = ScalarParameter() {defaultValue=0.25, group=Grass_Desert, parameterName=Grass_Desert_Chap_RoughAdd}
  n8 = TextureSample(
   UVs=@n2.Result
   Tex=TextureObject() {samplerType=SAMPLERTYPE_Normal, texture=/Game/Environments/Helios/Terrain/Textures/T_Helios_Chap_CoastGround_01_N}
  ) {
   samplerSource=SSM_Wrap_WorldGroupSettings
   samplerType=SAMPLERTYPE_Normal
   texture=&T_Helios_Chap_CoastGround_01_N
  }
  n9 = SetMaterialAttributes!(
   MaterialAttributes
   BaseColor=Saturate(None=Multiply(A=@n1.RGB, B=@n3.RGB))
   Specular=@n5.R
   Roughness=Saturate(None=Add(A=@n6, B=@n7))
   Normal=MaterialFunctionCall!(Normals=@n8.RGB, TransformToTangentSpace, Normal Strength=Constant() {r=1}) {materialFunction=/Game/Athena/Apollo/Environments/Landscape/Materials/Biome/CommonMaterials/MF_HQ_WorldAlignedNormal}.Result
   OpacityMask=@n3.A
   Displacement=@n3.A
  ) {
   desc=HighEnd Material
   attributeSetTypes [
    #1
    #2
    #3
    #4
    #5
    #6
   ]
  }
  n10 = SetMaterialAttributes(
   BaseColor=NamedRerouteUsage()
   Specular=NamedRerouteUsage()
   Roughness=NamedRerouteUsage()
   Normal=NamedRerouteUsage()
   OpacityMask=NamedRerouteUsage()
   Displacement=NamedRerouteUsage()
  ) {
   desc=HighEnd Material
   attributeSetTypes [
    #1
    #2
    #3
    #4
    #5
    #6
   ]
  }
  n11 = MaterialFunctionCall!(InputMaterial=@n9, LayerMaterial=@n10, LayerWeight=NamedRerouteUsage(), UseContrast, Contrast) {materialFunction=/Game/Athena/Helios/Environments/Landscape/Materials/MaterialFunctions_Utilities/MF_HeightBlendLandscapeLayers_OpacityMask}
  FunctionOutput(None=@n11.BlendedMaterial) {sortPriority=-500}
  n12 = Constant()
  FunctionOutput(None=Multiply(
    A=FunctionInput(Preview=@n12) {bUsePreviewValueAsDefault=true, inputName=LayerWeight, inputType=FunctionInput_Scalar, sortPriority=-500}
    B=NamedRerouteUsage()
   )) {outputName=HighEnd_LayerWeight AND DesertGrass_Noise, sortPriority=-400}
  FunctionOutput(None=@n11.BlendedMaterial) {outputName=StaticMeshMaterial, sortPriority=-600}
  FunctionOutput(None=@n11.LayerMask) {outputName=DesertGrass_Noise, sortPriority=-300}
  n13 = NamedRerouteUsage()
  NamedRerouteDeclaration(None=LinearInterpolate(B=@n13, Alpha=ScalarParameter() {defaultValue=1, group=Displacement, parameterName=Grass_Desert_DisplacementAmt}) {constA=0.5}) {name=Displacement, nodeColor=<rgba 0.5 0 1 1>}
  NamedRerouteDeclaration(None=@n5.G) {name=FlowerMask, nodeColor=<rgba 0.5 0 1 1>}
  n14 = MaterialFunctionCall(Scale=ScalarParameter() {defaultValue=900, group=Grass_Desert, parameterName=Grass_Desert_WorldScale}) {materialFunction=&MF_SimpleWorldUV}
  n15 = TextureSample(UVs=@n14.Result) {
   samplerSource=SSM_Wrap_WorldGroupSettings
   samplerType=SAMPLERTYPE_Normal
   texture=/Game/Environments/Helios/Terrain/Textures/T_Helios_RustBucket_Arid_N
  }
  n16 = FunctionInput(Preview=StaticBool()) {
   bUsePreviewValueAsDefault=true
   description="Default = False"
   inputName=UseTangentNormals
   inputType=FunctionInput_StaticBool
   sortPriority=1
  }
  n17 = MaterialFunctionCall(Normals=@n15.RGB, TransformToTangentSpace=@n16, Normal Strength=Constant() {r=1}) {materialFunction=&MF_HQ_WorldAlignedNormal}
  NamedRerouteDeclaration(None=@n17.Result) {name=Normal_SM, nodeColor=<rgba 0.5 0 1 1>}
  n18 = TextureObject() {texture=/Game/Environments/Helios/Terrain/Textures/T_Helios_Landscape_Noise}
  n19 = TextureSample(
   UVs=MaterialFunctionCall(Scale=ScalarParameter() {defaultValue=3000, group=Grass_Desert, parameterName=Grass_Desert_Noise01_Scale}) {materialFunction=&MF_SimpleWorldUV}.Result
   Tex=@n18
  ) {samplerSource=SSM_Wrap_WorldGroupSettings, texture=/Game/Global/Textures/Default/Noise/T_Tiling_Noise_01}
  NamedRerouteDeclaration(None=Power(Base=@n19.R, Exp=ScalarParameter() {defaultValue=0.4, group=Grass_Desert, parameterName=Grass_Desert_Noise01_Power})) {name=BlendNoise, nodeColor=<rgba 1 0 0.063 1>}
  NamedRerouteDeclaration(None=@n17.Result) {name=Normal, nodeColor=<rgba 0.5 0 1 1>}
  n20 = TextureSample(UVs=@n14.Result) {samplerSource=SSM_Wrap_WorldGroupSettings, texture=/Game/Environments/Helios/Terrain/Textures/T_Helios_RustBucket_Arid_D}
  NamedRerouteDeclaration(None=@n20.A) {name=Grass_DesertGrass_Height, nodeColor=<rgba 0.5 0 1 1>}
  NamedRerouteDeclaration(None=@n20.RGB) {name=Grass_DesertGrass_Diffuse, nodeColor=<rgba 0.5 0 1 1>}
  n21 = ScalarParameter() {defaultValue=500, group=Grass_Desert, parameterName=Grass_Desert_Noise02_Scale}
  n22 = Multiply(
   A=@n19.B
   B=TextureSample(
    UVs=MaterialFunctionCall(Scale=@n21) {materialFunction=&MF_SimpleWorldUV}.Result
    Tex=@n18
   ) {samplerSource=SSM_Wrap_WorldGroupSettings, texture=&T_Tiling_Noise_01}.B
  )
  n23 = Multiply(
   A=OneMinus(None=@n19.B)
   B=MaterialFunctionCall!(Thickness=SquareRoot(None=@n22), Depth Scale=Constant() {r=1}) {materialFunction=/Engine/Functions/Engine_MaterialFunctions01/Volumetrics/BeersLaw}.Result
  )
  NamedRerouteDeclaration(None=Multiply(A=@n23, B=@n23)) {name=LandscapeLayerMaskNoise, nodeColor=<rgba 1 0 0.063 1>}
  n24 = TextureSample(UVs=@n14.Result) {
   samplerSource=SSM_Wrap_WorldGroupSettings
   samplerType=SAMPLERTYPE_Masks
   texture=/Game/Environments/Helios/Terrain/Textures/T_Helios_RustBucket_Arid_S
  }
  n25 = MaterialFunctionCall(Float3=@n24.RGB) {materialFunction=&BreakOutFloat3Components}
  NamedRerouteDeclaration(None=@n25.R) {name=Specular, nodeColor=<rgba 0.5 0 1 1>}
  NamedRerouteDeclaration(None=@n25.B) {name=Roughness, nodeColor=<rgba 0.5 0 1 1>}
  NamedRerouteDeclaration(None=@n20.A) {name=Grass_DesertGrass_Height_SM, nodeColor=<rgba 0.5 0 1 1>}
  NamedRerouteDeclaration(None=@n20.RGB) {name=Grass_DesertGrass_Diffuse_SM, nodeColor=<rgba 0.5 0 1 1>}
  n26 = MaterialFunctionCall(Float3=@n24.RGB) {materialFunction=&BreakOutFloat3Components}
  NamedRerouteDeclaration(None=@n26.R) {name=Specular_SM, nodeColor=<rgba 0.5 0 1 1>}
  NamedRerouteDeclaration(None=@n26.B) {name=Roughness_SM, nodeColor=<rgba 0.5 0 1 1>}
  Power(Base=@n23, Exp=ScalarParameter() {defaultValue=2, group=Grass_Desert, parameterName=Grass_Desert_Noise_Power})
  n27 = ScalarParameter() {defaultValue=3, group=Displacement, parameterName=Grass_Desert_DisplacementPower}
  Saturate(None=Multiply(A=Power(Base=@n13, Exp=@n27), B=ScalarParameter() {defaultValue=2, group=Displacement, parameterName=Grass_Desert_DisplacementMod}))
  ScalarParameter() {defaultValue=1, group=Grass_Desert, parameterName=Grass_Desert_Noise_BeersLaw}
  SetMaterialAttributes!(
   MaterialAttributes
   BaseColor=NamedRerouteUsage()
   Specular=NamedRerouteUsage()
   Roughness=NamedRerouteUsage()
   Normal=NamedRerouteUsage()
   OpacityMask=NamedRerouteUsage()
  ) {
   desc=Static Mesh Material
   attributeSetTypes [
    #1
    #2
    #3
    #4
    #5
   ]
  }
 ]}
