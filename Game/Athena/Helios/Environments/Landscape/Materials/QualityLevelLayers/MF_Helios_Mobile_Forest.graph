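# Material function that tints the boreal forest landscape material for mobile in the Helios environment.
# It takes a MaterialAttributes input, multiplies its BaseColor by the Mobile Forest Color vector parameter, and writes the result back into the attributes.
# All other attributes pass through unchanged, and the default tint shifts the color toward green-cyan.
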MaterialFunction /Game/Athena/Helios/Environments/Landscape/Materials/QualityLevelLayers/MF_Helios_Mobile_Forest
graph {nodes [
  n1 = GetMaterialAttributes(None=FunctionInput() {bUsePreviewValueAsDefault=true, inputName=MobileBorealForestMaterial, inputType=FunctionInput_MaterialAttributes}) {attributeGetTypes [#1]}
  n2 = VectorParameter() {group=Mobile Forest, parameterName=Mobile Forest Color, defaultValue=<rgba 0.7125 1.2 0.9842 1>}
  FunctionOutput(None=SetMaterialAttributes!(MaterialAttributes=@n1.MaterialAttributes, BaseColor=Multiply(A=@n1.BaseColor, B=@n2.RGB)) {attributeSetTypes [#1]})
  Reroute()
 ]}
